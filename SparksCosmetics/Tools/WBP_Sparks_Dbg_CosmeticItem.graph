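# Debug widget row for a Sparks cosmetic picker that selects one cosmetic by subtype and instrument, plus its variants.
# SetTypes fills the combo box from the cosmetic catalog; selecting an entry runs the Sparks.SetCosmetic console command.
# RefreshChoices reads the local pawn's equipped item and rebuilds a child widget per variant channel, restoring choices from the parent picker.
# ChildSetChoice forwards variant picks to the picker's cosmetic component.
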WidgetBlueprint /SparksCosmetics/Tools/WBP_Sparks_Dbg_CosmeticItem
graphs [
 ChildSetChoice {nodes [
   n1 = ChildSetChoice!() -> (then, Channel:Gameplay Tag Structure, Choice:String) {is=title K2Node_FunctionEntry}
   > SetVariantChoice!(
    execute
    self:Sparks Cosmetic Component Object Reference=Get CosmeticComponent!(self:WBP Sparks Dbg Cosmetic Picker Object Reference=Get ParentPicker() -> (ParentPicker:WBP Sparks Dbg Cosmetic Picker Object Reference)) -> (CosmeticComponent:Sparks Cosmetic Component Object Reference)
    Subtype:ESparksAccountItemSubtype Enum=Byte to Enum ESparksAccountItemSubtype!(Byte:Byte=Get Subtype() -> (Subtype:ESparksAccountItemSubtype Enum=Aura)) -> (ReturnValue:ESparksAccountItemSubtype Enum) {is=title K2Node_CastByteToEnum}
    InstrumentType:ESparksInstrumentType Enum=Get Instrument() -> (Instrument:ESparksInstrumentType Enum=Guitar)
    VariantChannel:Gameplay Tag Structure=@n1
    Value:String=@n1.Choice
   ) -> (then)
  ]}
 EventGraph {nodes [
   n1 = "On Selection Changed (cb_ItemList)"!() -> (OutputDelegate, then, SelectedItem:String, SelectionType:ESelectInfo Enum) {is=title K2Node_ComponentBoundEvent}
   > ExecuteConsoleCommand!(
    execute
    Command="To String (Text)"(InText=Format Text!(Format="LOCTEXT(\"Sparks.SetCosmetic {item}\")", item:String=@n1.SelectedItem) -> (Result))
    SpecificPlayer
   ) -> (then)
   > RefreshChoices()
  ]}
 RefreshChoices {nodes [
   RefreshChoices() {is=title K2Node_FunctionEntry}
   > n1 = Sequence!(execute) -> (then_0, then_1, then_2:Exec, then_3:Exec)
   n2 = GetComponentByClass!(
    self:Actor Object Reference=GetPlayerPawn!(self:Fort Player Controller Object Reference=GetFirstLocalFortPlayerController!(bIncludeNonSpectators:bool=true, bIncludeSpectators:bool=false) -> (ReturnValue:Fort Player Controller Object Reference)) -> (ReturnValue:Fort Player Pawn Object Reference)
    ComponentClass:Actor Component Class Reference=/SparksCosmetics/BP_SparksCosmeticComponent.BP_SparksCosmeticComponent_C
   ) -> (ReturnValue:BP Sparks Cosmetic Component Object Reference)
   > Is Valid(InputObject=@n2)
   > n3 = Set AccountItem(AccountItem:Sparks Account Item Definition Object Reference=GetCosmeticItem!(
     self:Sparks Cosmetic Component Object Reference=@n2
     SubType:ESparksAccountItemSubtype Enum=Get Subtype() -> (Subtype=Aura)
     Instrument:ESparksInstrumentType Enum=Get Instrument() -> (Instrument=Guitar)
    ) -> (ReturnValue:Sparks Account Item Definition Object Reference)) -> (Output_Get:Sparks Account Item Definition Object Reference)
   > n4 = Is Valid(InputObject=@n3)
   n5 = Get cb_ItemList() -> (cb_ItemList:"Combo Box (String) Object Reference")
   SetSelectedIndex!(execute=@n4.Is Not Valid, self:"Combo Box (String) Object Reference"=@n5, Index:int=-1) -> (then)
   SetSelectedOption!(
    execute=@n4.Is Valid
    self:"Combo Box (String) Object Reference"=@n5
    Option:String="To String (Text)"(InText=GetDisplayName!(self:Item Definition Base Object Reference=@n3) -> (ReturnValue:Text))
   ) -> (then)
   n6 = Get vb_Channels() -> (vb_Channels:Vertical Box Object Reference)
   n7 = For Each Loop(Exec=@n1.then_1, Array:Array of Widget Object References=GetAllChildren(self=@n6)) -> (Array Element:Widget Object Reference)
   > RemoveChild(self=@n6, Content=@n7)
   > RemoveFromParent(self=@n7)
   Set HasChannelWithOptions(execute=@n1.then_2, HasChannelWithOptions:bool=false) -> (Output_Get:bool)
   n8 = Get AccountItem() -> (AccountItem:Sparks Account Item Definition Object Reference)
   > Is Valid(InputObject=@n8)
   > n9 = GetAllFortVariantTokens!(execute) -> (then, ReturnValue:Array of Fort Variant Token Type Object References)
   > n10 = GetVariantChannels!(
    execute
    ItemDefinition:Athena Cosmetic Item Definition Object Reference=@n8
    VariantTokens:Array of Fort Variant Token Type Object References=@n9
    UnlockableOnly:bool=false
   ) -> (then, ReturnValue:Array of Gameplay Tag Structures)
   > Set ChannelCount(ChannelCount:int=Length(TargetArray:Array of Gameplay Tag Structures=@n10)) -> (Output_Get:int=0)
   > n11 = For Each Loop(Array:Array of Gameplay Tag Structures=@n10) -> (Array Element:Gameplay Tag Structure)
   > n12 = Create Widget!(
    execute:Exec
    Class:User Widget Class Reference=&/WBP_Sparks_Dbg_VariantChannel.WBP_Sparks_Dbg_VariantChannel_C
    OwningPlayer:Player Controller Object Reference
   ) -> (then:Exec, ReturnValue:WBP Sparks Dbg Variant Channel Object Reference)
   > AddChild(self=Get vb_Channels(), Content=@n12)
   > Set ParentItem!(
    execute
    ParentItem:WBP Sparks Dbg Cosmetic Item Object Reference=Self-Reference()
    self:WBP Sparks Dbg Variant Channel Object Reference=@n12
   ) -> (then, Output_Get:WBP Sparks Dbg Cosmetic Item Object Reference)
   n13 = GetVariantChannelOptionNames!(self:Sparks Account Item Definition Object Reference=@n8, ChannelTag:Gameplay Tag Structure=@n11) -> (OutChannelName:Text, OutNames:Array of Strings)
   n14 = Length(TargetArray:Array of Strings=@n13.OutNames)
   > Set HasChannelWithOptions(HasChannelWithOptions=OR Boolean(A="integer > integer"(A=@n14, B=1), B=Get HasChannelWithOptions() -> (HasChannelWithOptions:bool)))
   > SetNameAndChannels!(
    execute
    self:WBP Sparks Dbg Variant Channel Object Reference=@n12
    Tag:Gameplay Tag Structure=@n11
    Channel Name:Text=@n13
    Options:Array of Strings=@n13.OutNames
   ) -> (then)
   > n15 = GetChoice!(
    execute
    self:WBP Sparks Dbg Cosmetic Picker Object Reference=Get ParentPicker()
    Item Subtype:ESparksAccountItemSubtype Enum=Get Subtype() -> (Subtype=Aura)
    Instrument type:ESparksInstrumentType Enum=Get Instrument() -> (Instrument=Guitar)
    Variant Channel:Gameplay Tag Structure=@n11
   ) -> (then, Choice:String)
   > SetSelection!(execute, self:WBP Sparks Dbg Variant Channel Object Reference=@n12, Choice:String=@n15) -> (then)
   n16 = Get ChannelCount() -> (ChannelCount:int=0)
   n17 = Branch(execute=@n1.then_3, Condition=OR Boolean(A=Get HasChannelWithOptions(), B="integer > integer"(A=@n16, B=1)))
   n18 = Get vb_Channels()
   > SetVisibility(self=@n18, InVisibility=Visible)
   SetVisibility(execute=@n17.else, self=@n18, InVisibility=Collapsed)
  ]}
 RefreshData {nodes [
   RefreshData!() -> (then, Choices:Map of Gameplay Tag Structures to Strings) {is=title K2Node_FunctionEntry}
   > n1 = Sequence!(execute) -> (then_0, then_1, then_2)
   > n2 = Switch on ESparksAccountItemSubtype!(execute, Selection:ESparksAccountItemSubtype Enum=Get Subtype() -> (Subtype=Aura)) -> (Aura:Exec, Guitar:Exec, Bass:Exec, Keyboard:Exec, Microphone:Exec, Drums:Exec, SpotlightAnim:Exec, Song:Exec) {is=title K2Node_SwitchEnum}
   n3 = Get txt_Name() -> (txt_Name:Common Text Object Reference)
   > Set Text(Text="INVTEXT(\"Aura (Current Session Only)\")", self=@n3)
   n4 = Get Instrument() -> (Instrument=Guitar)
   Set Text(
    execute=[@n2.Bass, @n2.Drums, @n2.Guitar, @n2.Keyboard, @n2.Microphone, @n2.SpotlightAnim]
    Text="To Text (String)"(InString=Enum to String!(Enumerator:ESparksInstrumentType Enum=@n4) -> (ReturnValue:String))
    self=@n3
   )
   n5 = GetSparksCosmeticCatalog!(execute=@n1.then_1) -> (then, ReturnValue:Sparks Cosmetic Catalog Object Reference)
   n6 = GetActiveItemDefNamesForSubtypeAndInstrument!(
    self:Sparks Cosmetic Catalog Object Reference=@n5
    SubType:ESparksAccountItemSubtype Enum=Get Subtype() -> (Subtype=Aura)
    Instrument:ESparksInstrumentType Enum=Get Instrument() -> (Instrument=Guitar)
   ) -> (ReturnValue:Array of Names)
   > n7 = Branch(Condition="integer > integer"(A=Length(TargetArray:Array of Names=@n6)))
   > n8 = For Each Loop(Array:Array of Names=@n6) -> (Array Element:Name)
   n9 = Get cb_ItemList()
   > AddOption!(execute, self:"Combo Box (String) Object Reference"=@n9, Option:String="To String (Name)"(InName=@n8)) -> (then)
   AddOption(execute=@n7.else, self=@n9, Option=None)
   n10 = Sequence(execute=@n1.then_2)
   n11 = GetComponentByClass(
    self=GetPlayerPawn(self=GetFirstLocalFortPlayerController(bIncludeNonSpectators=true, bIncludeSpectators=false))
    ComponentClass=&BP_SparksCosmeticComponent_C
   )
   > Is Valid(InputObject=@n11)
   > n12 = Set AccountItem(AccountItem=GetCosmeticItem(self=@n11, SubType=Get Subtype() -> (Subtype=Aura), Instrument=Get Instrument() -> (Instrument=Guitar)))
   > n13 = Is Valid(InputObject=@n12)
   n14 = Get cb_ItemList()
   SetSelectedIndex(execute=@n13.Is Not Valid, self=@n14, Index=-1)
   SetSelectedOption(execute=@n13.Is Valid, self=@n14, Option="To String (Text)"(InText=GetDisplayName(self=@n12)))
   ClearChildren(execute=@n10.then_1, self=Get vb_Channels())
   n15 = Get AccountItem()
   > Is Valid(InputObject=@n15)
   > n16 = GetAllFortVariantTokens()
   > n17 = GetVariantChannels(ItemDefinition=@n15, VariantTokens=@n16, UnlockableOnly=false)
   > n18 = For Each Loop(Array:Array of Gameplay Tag Structures=@n17) -> (Array Element:Gameplay Tag Structure)
   > n19 = Create Widget(Class=&WBP_Sparks_Dbg_VariantChannel_C)
   > AddChild(self=Get vb_Channels(), Content=@n19)
   > Set ParentItem(ParentItem=Self-Reference(), self=@n19)
   n20 = GetVariantChannelOptionNames(self=@n15, ChannelTag=@n18)
   > SetNameAndChannels(self=@n19, Tag=@n18, Channel Name=@n20, Options=@n20.OutNames)
   RefreshChoices(execute=@n10.then_2) {is=title K2Node_CallFunction}
  ]}
 SetItemInfo {nodes [
   n1 = SetItemInfo!() -> (then, Item:Sparks Account Item Definition Object Reference) {is=title K2Node_FunctionEntry}
   > n2 = Set AccountItem(AccountItem=@n1)
   n3 = Get SubType!(self:Sparks Account Item Definition Object Reference=@n2) -> (SubType:ESparksAccountItemSubtype Enum=Aura)
   > Set Text(Text="To Text (String)"(InString=Enum to String(Enumerator:ESparksAccountItemSubtype Enum=@n3)), self=Get txt_Name())
   > RefreshChoices()
  ]}
 SetTypes {nodes [
   n1 = SetTypes!() -> (then, Subtype:ESparksAccountItemSubtype Enum=Aura, Instrument:ESparksInstrumentType Enum=Guitar) {is=title K2Node_FunctionEntry}
   > Set Subtype(Subtype:ESparksAccountItemSubtype Enum=@n1) -> (Output_Get:ESparksAccountItemSubtype Enum=Aura)
   > n2 = Set Instrument(Instrument:ESparksInstrumentType Enum=@n1.Instrument) -> (Output_Get:ESparksInstrumentType Enum=Guitar)
   > Set Text(Text="To Text (String)"(InString=Enum to String(Enumerator:ESparksInstrumentType Enum=@n2)), self=Get txt_Name())
   > RefreshData!(
    execute
    self
    Choices:Map of Gameplay Tag Structures to Strings=Get EmptyChoices() -> (EmptyChoices:Map of Gameplay Tag Structures to Strings)
   ) -> (then) {is=title K2Node_CallFunction}
  ]}
]
variables [
 AccountItem {type=SparksAccountItemDefinition, default=null}
 Instrument {type=ESparksInstrumentType, default=Guitar}
 ParentPicker {type=WBP_Sparks_Dbg_CosmeticPicker_C, default=null}
 Subtype {type=ESparksAccountItemSubtype, default=Aura}
]
widgets {
 info {inheritedWidgetCount=0, parentClass=&UserWidget, rootWidgetClass=&CanvasPanel, widgetCount=7}
 widgets [
  CanvasPanel_133 {widgetClassPath=&CanvasPanel, widget=CanvasPanel_133}
  VerticalBox_0 {
   widgetClassPath=&VerticalBox
   parent=CanvasPanel_133
   widget=VerticalBox_0
   slot {
    class=CanvasPanelSlot
    properties {bAutoSize=true, layoutData {alignment=<0 0>, anchors {maximum=<1 1>, minimum=<0 0>}, offsets=<ltrb 0 0 0 0>}}
   }
  }
  HorizontalBox_0 {widgetClassPath=&HorizontalBox, parent=VerticalBox_0, widget=HorizontalBox_0, slot {class=VerticalBoxSlot}}
  txt_Name {
   widgetClassPath=&CommonTextBlock
   bIsVariable=true
   parent=HorizontalBox_0
   widget=txt_Name
   properties {
    style=/FortUILibrary/Styles/Typography/_TextStyle_SubH4._TextStyle_SubH4_C
    text=Cosmetic Item
    font {
     bForceMonospaced=false
     bMaterialIsStencil=false
     fontMaterial=null
     fontObject=&BurbankSmall
     letterSpacing=0
     monospacedWidth=1
     size=24
     skewAmount=0
     typefaceFontName=Black
     outlineSettings {
      bApplyOutlineToDropShadows=false
      bMiteredCorners=false
      bSeparateFillAlpha=false
      outlineMaterial=null
      outlineSize=0
      outlineColor=<rgba 0 0 0 1>
     }
    }
    shadowOffset=<0 0>
    strikeBrush {
     bIsDynamicallyLoaded=false
     drawAs=NoDrawType
     imageType=NoImage
     mirroring=NoMirror
     resourceName=None
     resourceObject=null
     tiling=NoTile
     imageSize=<32 32>
     margin=<ltrb 0 0 0 0>
     outlineSettings {
      bUseBrushTransparency=false
      roundingType=HalfHeightRadius
      width=0
      color {colorUseRule=UseColor_Specified, specifiedColor=<rgba 0 0 0 0>}
      cornerRadii=<0 0 0 0>
     }
     tintColor {colorUseRule=UseColor_Specified, specifiedColor=<rgba 1 1 1 1>}
     uVRegion {bIsValid=false, max=<0 0>, min=<0 0>}
    }
   }
   slot {class=HorizontalBoxSlot}
  }
  cb_ItemList {
   widgetClassPath=/Script/UMG.ComboBoxString
   bIsVariable=true
   parent=HorizontalBox_0
   widget=cb_ItemList
   properties {
    font {
     bForceMonospaced=false
     bMaterialIsStencil=false
     fontMaterial=null
     fontObject=&Roboto
     letterSpacing=0
     monospacedWidth=1
     size=16
     skewAmount=0
     typefaceFontName=Bold
     outlineSettings {
      bApplyOutlineToDropShadows=false
      bMiteredCorners=false
      bSeparateFillAlpha=false
      outlineMaterial=null
      outlineSize=1
      outlineColor=<rgba 0.2188 0.2188 0.2188 1>
     }
    }
    itemStyle {
     bUseParentRowBrush=false
     activeBrush {
      bIsDynamicallyLoaded=false
      drawAs=Image
      imageType=FullColor
      mirroring=NoMirror
      resourceName=../../../Engine/Content/Slate/Common/Selection.png
      resourceObject=null
      tiling=NoTile
      imageSize=<8 8>
      margin=<ltrb 0 0 0 0>
      outlineSettings {
       bUseBrushTransparency=false
       roundingType=HalfHeightRadius
       width=0
       color {colorUseRule=UseColor_Specified, specifiedColor=<rgba 0 0 0 0>}
       cornerRadii=<0 0 0 0>
      }
      tintColor {colorUseRule=UseColor_Specified, specifiedColor=<rgba 0.0091 0.3515 0.9647 1>}
      uVRegion {bIsValid=false, max=<0 0>, min=<0 0>}
     }
     activeHighlightedBrush {
      bIsDynamicallyLoaded=false
      drawAs=Image
      imageType=FullColor
      mirroring=NoMirror
      resourceName=../../../Engine/Content/Slate/Common/Selection.png
      resourceObject=null
      tiling=NoTile
      imageSize=<8 8>
      margin=<ltrb 0 0 0 0>
      outlineSettings {
       bUseBrushTransparency=false
       roundingType=HalfHeightRadius
       width=0
       color {colorUseRule=UseColor_Specified, specifiedColor=<rgba 0 0 0 0>}
       cornerRadii=<0 0 0 0>
      }
      tintColor {colorUseRule=UseColor_Specified, specifiedColor=<rgba 0 0.1878 0.6795 1>}
      uVRegion {bIsValid=false, max=<0 0>, min=<0 0>}
     }
     activeHoveredBrush {
      bIsDynamicallyLoaded=false
      drawAs=Image
      imageType=FullColor
      mirroring=NoMirror
      resourceName=../../../Engine/Content/Slate/Common/Selection.png
      resourceObject=null
      tiling=NoTile
      imageSize=<8 8>
      margin=<ltrb 0 0 0 0>
      outlineSettings {
       bUseBrushTransparency=false
       roundingType=HalfHeightRadius
       width=0
       color {colorUseRule=UseColor_Specified, specifiedColor=<rgba 0 0 0 0>}
       cornerRadii=<0 0 0 0>
      }
      tintColor {colorUseRule=UseColor_Specified, specifiedColor=<rgba 0.0091 0.3515 0.9647 1>}
      uVRegion {bIsValid=false, max=<0 0>, min=<0 0>}
     }
     dropIndicator_Above {
      bIsDynamicallyLoaded=false
      drawAs=Box
      imageType=FullColor
      mirroring=NoMirror
      resourceName=../../../Engine/Content/Slate/Common/DropZoneIndicator_Above.png
      resourceObject=null
      tiling=NoTile
      imageSize=<0 0>
      margin=<ltrb 0.625 0.625 0 0>
      outlineSettings {
       bUseBrushTransparency=false
       roundingType=HalfHeightRadius
       width=0
       color {colorUseRule=UseColor_Specified, specifiedColor=<rgba 0 0 0 0>}
       cornerRadii=<0 0 0 0>
      }
      tintColor {colorUseRule=UseColor_Specified, specifiedColor=<rgba 0.0091 0.3515 0.9647 1>}
      uVRegion {bIsValid=false, max=<0 0>, min=<0 0>}
     }
     dropIndicator_Below {
      bIsDynamicallyLoaded=false
      drawAs=Box
      imageType=FullColor
      mirroring=NoMirror
      resourceName=../../../Engine/Content/Slate/Common/DropZoneIndicator_Below.png
      resourceObject=null
      tiling=NoTile
      imageSize=<0 0>
      margin=<ltrb 0.625 0 0 0.625>
      outlineSettings {
       bUseBrushTransparency=false
       roundingType=HalfHeightRadius
       width=0
       color {colorUseRule=UseColor_Specified, specifiedColor=<rgba 0 0 0 0>}
       cornerRadii=<0 0 0 0>
      }
      tintColor {colorUseRule=UseColor_Specified, specifiedColor=<rgba 0.0091 0.3515 0.9647 1>}
      uVRegion {bIsValid=false, max=<0 0>, min=<0 0>}
     }
     dropIndicator_Onto {
      bIsDynamicallyLoaded=false
      drawAs=Box
      imageType=FullColor
      mirroring=NoMirror
      resourceName=../../../Engine/Content/Slate/Common/DropZoneIndicator_Onto.png
      resourceObject=null
      tiling=NoTile
      imageSize=<0 0>
      margin=<ltrb 0.25 0.25 0.25 0.25>
      outlineSettings {
       bUseBrushTransparency=false
       roundingType=HalfHeightRadius
       width=0
       color {colorUseRule=UseColor_Specified, specifiedColor=<rgba 0 0 0 0>}
       cornerRadii=<0 0 0 0>
      }
      tintColor {colorUseRule=UseColor_Specified, specifiedColor=<rgba 0.0091 0.3515 0.9647 1>}
      uVRegion {bIsValid=false, max=<0 0>, min=<0 0>}
     }
     evenRowBackgroundBrush {
      bIsDynamicallyLoaded=false
      drawAs=NoDrawType
      imageType=NoImage
      mirroring=NoMirror
      resourceName=None
      resourceObject=null
      tiling=NoTile
      imageSize=<0 0>
      margin=<ltrb 0 0 0 0>
      outlineSettings {
       bUseBrushTransparency=false
       roundingType=HalfHeightRadius
       width=0
       color {colorUseRule=UseColor_Specified, specifiedColor=<rgba 0 0 0 0>}
       cornerRadii=<0 0 0 0>
      }
      tintColor {colorUseRule=UseColor_Specified, specifiedColor=<rgba 1 1 1 1>}
      uVRegion {bIsValid=false, max=<0 0>, min=<0 0>}
     }
     evenRowBackgroundHoveredBrush {
      bIsDynamicallyLoaded=false
      drawAs=Image
      imageType=FullColor
      mirroring=NoMirror
      resourceName=../../../Engine/Content/Slate/Common/Selection.png
      resourceObject=null
      tiling=NoTile
      imageSize=<8 8>
      margin=<ltrb 0 0 0 0>
      outlineSettings {
       bUseBrushTransparency=false
       roundingType=HalfHeightRadius
       width=0
       color {colorUseRule=UseColor_Specified, specifiedColor=<rgba 0 0 0 0>}
       cornerRadii=<0 0 0 0>
      }
      tintColor {colorUseRule=UseColor_Specified, specifiedColor=<rgba 1 1 1 0.1>}
      uVRegion {bIsValid=false, max=<0 0>, min=<0 0>}
     }
     inactiveBrush {
      bIsDynamicallyLoaded=false
      drawAs=Image
      imageType=FullColor
      mirroring=NoMirror
      resourceName=../../../Engine/Content/Slate/Common/Selection.png
      resourceObject=null
      tiling=NoTile
      imageSize=<8 8>
      margin=<ltrb 0 0 0 0>
      outlineSettings {
       bUseBrushTransparency=false
       roundingType=HalfHeightRadius
       width=0
       color {colorUseRule=UseColor_Specified, specifiedColor=<rgba 0 0 0 0>}
       cornerRadii=<0 0 0 0>
      }
      tintColor {colorUseRule=UseColor_Specified, specifiedColor=<rgba 0.25 0.25 0.25 1>}
      uVRegion {bIsValid=false, max=<0 0>, min=<0 0>}
     }
     inactiveHighlightedBrush {
      bIsDynamicallyLoaded=false
      drawAs=Image
      imageType=FullColor
      mirroring=NoMirror
      resourceName=../../../Engine/Content/Slate/Common/Selection.png
      resourceObject=null
      tiling=NoTile
      imageSize=<8 8>
      margin=<ltrb 0 0 0 0>
      outlineSettings {
       bUseBrushTransparency=false
       roundingType=HalfHeightRadius
       width=0
       color {colorUseRule=UseColor_Specified, specifiedColor=<rgba 0 0 0 0>}
       cornerRadii=<0 0 0 0>
      }
      tintColor {colorUseRule=UseColor_Specified, specifiedColor=<rgba 0.1 0.1 0.1 1>}
      uVRegion {bIsValid=false, max=<0 0>, min=<0 0>}
     }
     inactiveHoveredBrush {
      bIsDynamicallyLoaded=false
      drawAs=Image
      imageType=FullColor
      mirroring=NoMirror
      resourceName=../../../Engine/Content/Slate/Common/Selection.png
      resourceObject=null
      tiling=NoTile
      imageSize=<8 8>
      margin=<ltrb 0 0 0 0>
      outlineSettings {
       bUseBrushTransparency=false
       roundingType=HalfHeightRadius
       width=0
       color {colorUseRule=UseColor_Specified, specifiedColor=<rgba 0 0 0 0>}
       cornerRadii=<0 0 0 0>
      }
      tintColor {colorUseRule=UseColor_Specified, specifiedColor=<rgba 0.25 0.25 0.25 1>}
      uVRegion {bIsValid=false, max=<0 0>, min=<0 0>}
     }
     oddRowBackgroundBrush {
      bIsDynamicallyLoaded=false
      drawAs=NoDrawType
      imageType=NoImage
      mirroring=NoMirror
      resourceName=None
      resourceObject=null
      tiling=NoTile
      imageSize=<0 0>
      margin=<ltrb 0 0 0 0>
      outlineSettings {
       bUseBrushTransparency=false
       roundingType=HalfHeightRadius
       width=0
       color {colorUseRule=UseColor_Specified, specifiedColor=<rgba 0 0 0 0>}
       cornerRadii=<0 0 0 0>
      }
      tintColor {colorUseRule=UseColor_Specified, specifiedColor=<rgba 1 1 1 1>}
      uVRegion {bIsValid=false, max=<0 0>, min=<0 0>}
     }
     oddRowBackgroundHoveredBrush {
      bIsDynamicallyLoaded=false
      drawAs=Image
      imageType=FullColor
      mirroring=NoMirror
      resourceName=../../../Engine/Content/Slate/Common/Selection.png
      resourceObject=null
      tiling=NoTile
      imageSize=<8 8>
      margin=<ltrb 0 0 0 0>
      outlineSettings {
       bUseBrushTransparency=false
       roundingType=HalfHeightRadius
       width=0
       color {colorUseRule=UseColor_Specified, specifiedColor=<rgba 0 0 0 0>}
       cornerRadii=<0 0 0 0>
      }
      tintColor {colorUseRule=UseColor_Specified, specifiedColor=<rgba 1 1 1 0.1>}
      uVRegion {bIsValid=false, max=<0 0>, min=<0 0>}
     }
     parentRowBackgroundBrush {
      bIsDynamicallyLoaded=false
      drawAs=Image
      imageType=NoImage
      mirroring=NoMirror
      resourceName=None
      resourceObject=null
      tiling=NoTile
      imageSize=<32 32>
      margin=<ltrb 0 0 0 0>
      outlineSettings {
       bUseBrushTransparency=false
       roundingType=HalfHeightRadius
       width=0
       color {colorUseRule=UseColor_Specified, specifiedColor=<rgba 0 0 0 0>}
       cornerRadii=<0 0 0 0>
      }
      tintColor {colorUseRule=UseColor_Specified, specifiedColor=<rgba 1 1 1 1>}
      uVRegion {bIsValid=false, max=<0 0>, min=<0 0>}
     }
     parentRowBackgroundHoveredBrush {
      bIsDynamicallyLoaded=false
      drawAs=Image
      imageType=NoImage
      mirroring=NoMirror
      resourceName=None
      resourceObject=null
      tiling=NoTile
      imageSize=<32 32>
      margin=<ltrb 0 0 0 0>
      outlineSettings {
       bUseBrushTransparency=false
       roundingType=HalfHeightRadius
       width=0
       color {colorUseRule=UseColor_Specified, specifiedColor=<rgba 0 0 0 0>}
       cornerRadii=<0 0 0 0>
      }
      tintColor {colorUseRule=UseColor_Specified, specifiedColor=<rgba 1 1 1 1>}
      uVRegion {bIsValid=false, max=<0 0>, min=<0 0>}
     }
     selectedTextColor {colorUseRule=UseColor_Specified, specifiedColor=<rgba 1 1 1 1>}
     selectorFocusedBrush {
      bIsDynamicallyLoaded=false
      drawAs=Border
      imageType=FullColor
      mirroring=NoMirror
      resourceName=../../../Engine/Content/Slate/Common/Selector.png
      resourceObject=null
      tiling=Both
      imageSize=<0 0>
      margin=<ltrb 0.25 0.25 0.25 0.25>
      outlineSettings {
       bUseBrushTransparency=false
       roundingType=HalfHeightRadius
       width=0
       color {colorUseRule=UseColor_Specified, specifiedColor=<rgba 0 0 0 0>}
       cornerRadii=<0 0 0 0>
      }
      tintColor {colorUseRule=UseColor_Specified, specifiedColor=<rgba 0.701 0.225 0.003 1>}
      uVRegion {bIsValid=false, max=<0 0>, min=<0 0>}
     }
     textColor {colorUseRule=UseColor_Specified, specifiedColor=<rgba 0.5156 0.5156 0.5156 1>}
    }
   }
   slot {class=HorizontalBoxSlot, properties {horizontalAlignment=HAlign_Right, size {sizeRule=Fill, value=1}}}
  }
  vb_Channels {
   widgetClassPath=&VerticalBox
   bIsVariable=true
   parent=VerticalBox_0
   widget=vb_Channels
   slot {class=VerticalBoxSlot, properties {padding=<ltrb 20 3 0 3>, size {sizeRule=Fill, value=1}}}
  }
  Image_201 {
   widgetClassPath=&Image
   bIsVariable=true
   parent=CanvasPanel_133
   widget=Image_201
   properties {brush {
     bIsDynamicallyLoaded=false
     drawAs=Image
     imageType=NoImage
     mirroring=NoMirror
     resourceName=None
     resourceObject=null
     tiling=NoTile
     imageSize=<32 32>
     margin=<ltrb 0 0 0 0>
     outlineSettings {
      bUseBrushTransparency=false
      roundingType=HalfHeightRadius
      width=0
      color {colorUseRule=UseColor_Specified, specifiedColor=<rgba 0 0 0 0>}
      cornerRadii=<0 0 0 0>
     }
     tintColor {colorUseRule=UseColor_Specified, specifiedColor=<rgba 0 0 0 0.1>}
     uVRegion {bIsValid=false, max=<0 0>, min=<0 0>}
    }}
   slot {
    class=CanvasPanelSlot
    properties {bAutoSize=true, zOrder=-5, layoutData {alignment=<0 0>, anchors {maximum=<1 1>, minimum=<0 0>}, offsets=<ltrb 2 2 2 2>}}
   }
  }
 ]
}
