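# Debug widget for the Sparks cosmetic picker, showing one collapsible group of cosmetic item rows for an account item subtype.
# SetGroupSubtype sets the title ("Instruments" for instrument subtypes) and calls RefreshAll when the subtype changes, except for Song.
# RefreshAll builds one CosmeticItem child per visible, non-None instrument, maps it by instrument and refreshes its choices.
# SetItemInfo forwards subtype-matching item definitions to the mapped child for that instrument.
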WidgetBlueprint /SparksCosmetics/Tools/WBP_Sparks_Dbg_Cosmetic_Group
graphs [
 EventGraph {nodes [
   n1 = "On Check State Changed (chk_Collapse)"!() -> (OutputDelegate, then, bIsChecked:bool) {is=title K2Node_ComponentBoundEvent}
   > SetVisibility(
    self=[Get vb_Items() -> (vb_Items:Vertical Box Object Reference), Get Image_0() -> (Image_0:Image Object Reference)]
    InVisibility=Select(Option 0=Collapsed, Option 1=Visible, Index=@n1.bIsChecked)
   )
  ]}
 OnChoiceChanged {nodes [
   n1 = OnChoiceChanged!() -> (then, Instrument:ESparksInstrumentType Enum=Guitar) {is=title K2Node_FunctionEntry}
   n2 = Find(
    TargetMap:Map of ESparksInstrumentType Enums to WBP Sparks Dbg Cosmetic Item Object References=Get Items() -> (Items:Map of ESparksInstrumentType Enums to WBP Sparks Dbg Cosmetic Item Object References=Guitar)
    Key:"ESparksInstrumentType Enum (by ref)"=@n1
   ) -> (Value:WBP Sparks Dbg Cosmetic Item Object Reference)
   > Branch(Condition=@n2.ReturnValue)
   > RefreshChoices(self:WBP Sparks Dbg Cosmetic Item Object Reference=@n2)
  ]}
 RefreshAll {nodes [
   RefreshAll() {is=title K2Node_FunctionEntry}
   > ClearChildren(self=Get vb_Items())
   > Clear!(execute, TargetMap:Map of ESparksInstrumentType Enums to WBP Sparks Dbg Cosmetic Item Object References=Get Items() -> (Items=Guitar)) -> (then) {is=title K2Node_CallFunction}
   > n1 = ForEach ESparksInstrumentType!(execute:Exec, SkipHidden:bool=true) -> (LoopBody:Exec, EnumValue:ESparksInstrumentType Enum, then:Exec) {is=title K2Node_ForEachElementInEnum}
   > n2 = Switch on ESparksInstrumentType!(execute, Selection:ESparksInstrumentType Enum=@n1) -> (Guitar:Exec, Bass:Exec, Vocals:Exec, Drum:Exec, Keyboard:Exec, None:Exec) {is=title K2Node_SwitchEnum}
   n3 = Create Widget!(
    execute:Exec=[@n2.Bass, @n2.Drum, @n2.Guitar, @n2.Keyboard, @n2.Vocals]
    Class:User Widget Class Reference=&/WBP_Sparks_Dbg_CosmeticItem.WBP_Sparks_Dbg_CosmeticItem_C
    OwningPlayer:Player Controller Object Reference
   ) -> (then:Exec, ReturnValue:WBP Sparks Dbg Cosmetic Item Object Reference)
   > Set ParentPicker!(
    execute
    ParentPicker:WBP Sparks Dbg Cosmetic Picker Object Reference=Get ParentPicker() -> (ParentPicker:WBP Sparks Dbg Cosmetic Picker Object Reference)
    self:WBP Sparks Dbg Cosmetic Item Object Reference=@n3
   ) -> (then, Output_Get:WBP Sparks Dbg Cosmetic Picker Object Reference)
   > SetTypes!(
    execute
    self:WBP Sparks Dbg Cosmetic Item Object Reference=@n3
    Subtype:ESparksAccountItemSubtype Enum=Get Item Type() -> (Item Type:ESparksAccountItemSubtype Enum=Aura)
    Instrument:ESparksInstrumentType Enum=@n1
   ) -> (then)
   > AddChild(self=Get vb_Items(), Content=@n3)
   > Add!(
    execute
    TargetMap:Map of ESparksInstrumentType Enums to WBP Sparks Dbg Cosmetic Item Object References=Get Items() -> (Items=Guitar)
    Key:"ESparksInstrumentType Enum (by ref)"=@n1
    Value:"WBP Sparks Dbg Cosmetic Item Object Reference (by ref)"=@n3
   ) -> (then) {is=title K2Node_CallFunction}
   > OnChoiceChanged!(execute, self, Instrument:ESparksInstrumentType Enum=@n1) -> (then) {is=title K2Node_CallFunction}
  ]}
 SetGroupSubtype {nodes [
   n1 = SetGroupSubtype!() -> (then, ItemType:ESparksAccountItemSubtype Enum=Aura) {is=title K2Node_FunctionEntry}
   > Branch(Condition="Not Equal (Enum)"!(A:ESparksAccountItemSubtype Enum=@n1, B:ESparksAccountItemSubtype Enum=Get Item Type() -> (Item Type=Aura)) -> (ReturnValue:bool))
   > n2 = Set Item Type(Item Type:ESparksAccountItemSubtype Enum=@n1) -> (Output_Get:ESparksAccountItemSubtype Enum=Aura)
   > n3 = Switch on ESparksAccountItemSubtype!(execute, Selection:ESparksAccountItemSubtype Enum=@n2) -> (Aura:Exec, Guitar:Exec, Bass:Exec, Keyboard:Exec, Microphone:Exec, Drums:Exec, SpotlightAnim:Exec, Song:Exec) {is=title K2Node_SwitchEnum}
   n4 = Get txt_Title() -> (txt_Title:Common Text Object Reference)
   n5 = Set Text(
    execute=[@n3.Aura, @n3.SpotlightAnim]
    Text="To Text (String)"(InString=Enum to String!(Enumerator:ESparksAccountItemSubtype Enum=@n2) -> (ReturnValue:String))
    self=@n4
   )
   RefreshAll(execute=[@n6.then, @n5.then]) {is=title K2Node_CallFunction}
   n6 = Set Text(execute=[@n3.Bass, @n3.Drums, @n3.Guitar, @n3.Keyboard, @n3.Microphone], Text="LOCTEXT(\"Instruments\")", self=@n4)
  ]}
 SetItemInfo {nodes [
   n1 = SetItemInfo!() -> (then, NewItem:Sparks Account Item Definition Object Reference) {is=title K2Node_FunctionEntry}
   > Is Valid(InputObject=@n1)
   > Branch(Condition="Equal (Enum)"(
     A:ESparksAccountItemSubtype Enum=Get SubType!(self:Sparks Account Item Definition Object Reference=@n1) -> (SubType:ESparksAccountItemSubtype Enum=Aura)
     B:ESparksAccountItemSubtype Enum=Get Item Type() -> (Item Type=Aura)
    ))
   n2 = Find(
    TargetMap=Get Items() -> (Items=Guitar)
    Key=GetInstrumentType!(self:Sparks Account Item Definition Object Reference=@n1) -> (ReturnValue:ESparksInstrumentType Enum=Guitar)
   )
   > Branch(Condition=@n2.ReturnValue)
   > SetItemInfo!(execute, self:WBP Sparks Dbg Cosmetic Item Object Reference=@n2, Item:Sparks Account Item Definition Object Reference=@n1) -> (then) {is=title K2Node_CallFunction}
   Branch()
   > SetItemInfo(
    self=Reroute Node(InputPin:WBP Sparks Dbg Cosmetic Item Object Reference) -> (OutputPin:WBP Sparks Dbg Cosmetic Item Object Reference)
    Item=Reroute Node(InputPin:Sparks Account Item Definition Object Reference) -> (OutputPin:Sparks Account Item Definition Object Reference)
   )
  ]}
]
variables [
 Item Type {type=ESparksAccountItemSubtype, default=Aura}
 Items {default {}}
 ParentPicker {type=WBP_Sparks_Dbg_CosmeticPicker_C, default=null}
 Title {type=boolean}
]
widgets {
 info {inheritedWidgetCount=0, parentClass=&UserWidget, rootWidgetClass=&CanvasPanel, widgetCount=6}
 widgets [
  CanvasPanel_82 {widgetClassPath=&CanvasPanel, widget=CanvasPanel_82}
  HorizontalBox_0 {
   widgetClassPath=&HorizontalBox
   parent=CanvasPanel_82
   widget=HorizontalBox_0
   slot {
    class=CanvasPanelSlot
    properties {bAutoSize=true, layoutData {alignment=<0 0>, anchors {maximum=<1 0>, minimum=<0 0>}, offsets=<ltrb 0 0 0 0>}}
   }
  }
  txt_Title {
   widgetClassPath=&CommonTextBlock
   bIsVariable=true
   parent=HorizontalBox_0
   widget=txt_Title
   properties {
    style=/FortUILibrary/Styles/Typography/_TextStyle_SubH3._TextStyle_SubH3_C
    text=Title
    font {
     bForceMonospaced=false
     bMaterialIsStencil=false
     fontMaterial=null
     fontObject=&Burbank
     letterSpacing=0
     monospacedWidth=1
     size=36
     skewAmount=0
     typefaceFontName=Burbank-PlayerPower
     outlineSettings {
      bApplyOutlineToDropShadows=false
      bMiteredCorners=false
      bSeparateFillAlpha=false
      outlineMaterial=null
      outlineSize=0
      outlineColor=<rgba 0 0 0 1>
     }
    }
    shadowOffset=<0 0>
    strikeBrush {
     bIsDynamicallyLoaded=false
     drawAs=NoDrawType
     imageType=NoImage
     mirroring=NoMirror
     resourceName=None
     resourceObject=null
     tiling=NoTile
     imageSize=<32 32>
     margin=<ltrb 0 0 0 0>
     outlineSettings {
      bUseBrushTransparency=false
      roundingType=HalfHeightRadius
      width=0
      color {colorUseRule=UseColor_Specified, specifiedColor=<rgba 0 0 0 0>}
      cornerRadii=<0 0 0 0>
     }
     tintColor {colorUseRule=UseColor_Specified, specifiedColor=<rgba 1 1 1 1>}
     uVRegion {bIsValid=false, max=<0 0>, min=<0 0>}
    }
   }
   slot {class=HorizontalBoxSlot}
  }
  chk_Collapse {
   widgetClassPath=/Script/UMG.CheckBox
   bIsVariable=true
   parent=HorizontalBox_0
   widget=chk_Collapse
   properties {checkedState=Checked}
   slot {class=HorizontalBoxSlot, properties {horizontalAlignment=HAlign_Right, padding=<ltrb 15 15 15 15>, size {sizeRule=Fill, value=1}}}
  }
  vb_Items {
   widgetClassPath=&VerticalBox
   bIsVariable=true
   parent=CanvasPanel_82
   widget=vb_Items
   slot {
    class=CanvasPanelSlot
    properties {bAutoSize=true, layoutData {alignment=<0 0>, anchors {maximum=<1 1>, minimum=<0 0>}, offsets=<ltrb 15 55 0 0>}}
   }
  }
  Image_0 {
   widgetClassPath=&Image
   bIsVariable=true
   parent=CanvasPanel_82
   widget=Image_0
   properties {brush {
     bIsDynamicallyLoaded=false
     drawAs=Image
     imageType=NoImage
     mirroring=NoMirror
     resourceName=None
     resourceObject=null
     tiling=NoTile
     imageSize=<32 32>
     margin=<ltrb 0 0 0 0>
     outlineSettings {
      bUseBrushTransparency=false
      roundingType=HalfHeightRadius
      width=0
      color {colorUseRule=UseColor_Specified, specifiedColor=<rgba 0 0 0 0>}
      cornerRadii=<0 0 0 0>
     }
     tintColor {colorUseRule=UseColor_Specified, specifiedColor=<rgba 0 0 0 0.1>}
     uVRegion {bIsValid=false, max=<0 0>, min=<0 0>}
    }}
   slot {
    class=CanvasPanelSlot
    properties {bAutoSize=true, layoutData {alignment=<0 0>, anchors {maximum=<1 1>, minimum=<0 0>}, offsets=<ltrb 2 2 2 2>}}
   }
  }
 ]
}
