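# Debug UMG panel for picking Sparks cosmetics on a Sparks Cosmetic Component.
# SetCosmeticComponent binds its choice and item change events, then ReloadData rebuilds an Aura item and an instrument group; the Spotlight group creation is unwired.
# Change handlers switch on subtype to refresh the matching child widget.
# Reload, Save and Defaults buttons run Sparks loadout console commands, and a checkbox toggles the item list's visibility.
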WidgetBlueprint /SparksCosmetics/Tools/WBP_Sparks_Dbg_CosmeticPicker
graphs [
 EventGraph {nodes [
   "On Button Clicked (btn_Reload)"!() -> (OutputDelegate, then, Button:Common Button Legacy Object Reference) {is=title K2Node_ComponentBoundEvent}
   > ExecuteConsoleCommand!(execute, Command=Sparks.LoadLoadout, SpecificPlayer) -> (then)
   n1 = "On Check State Changed (chk_Collapse)"!() -> (OutputDelegate, then, bIsChecked:bool) {is=title K2Node_ComponentBoundEvent}
   > SetVisibility(
    self=Get vb_Items() -> (vb_Items:Vertical Box Object Reference)
    InVisibility=Select(Option 0=Collapsed, Option 1=Visible, Index=@n1.bIsChecked)
   )
   "On Button Clicked (btn_Defaults)"!() -> (OutputDelegate, then, Button:Common Button Legacy Object Reference) {is=title K2Node_ComponentBoundEvent}
   > ExecuteConsoleCommand(Command=Sparks.ClearLoadout)
   "On Button Clicked (btn_Save)"!() -> (OutputDelegate, then, Button:Common Button Legacy Object Reference) {is=title K2Node_ComponentBoundEvent}
   > ExecuteConsoleCommand(Command=Sparks.SaveLoadout)
   n2 = Event SetCosmeticComponent!() -> (OutputDelegate, then, InCosmeticComponent:Sparks Cosmetic Component Object Reference) {is=title K2Node_Event}
   > Parent: SetCosmeticComponent!(execute, InCosmeticComponent:Sparks Cosmetic Component Object Reference=@n2.InCosmeticComponent) -> (then) {is=title K2Node_CallParentFunction}
   > n3 = Set CosmeticComponent(CosmeticComponent:Sparks Cosmetic Component Object Reference=@n2.InCosmeticComponent) -> (Output_Get:Sparks Cosmetic Component Object Reference)
   > n4 = Is Valid(InputObject=@n3)
   > Bind Event to Cosmetic Choice Changed(self:Sparks Cosmetic Component Object Reference=@n3, Delegate=Create Event()) {is=title K2Node_AddDelegate}
   > n5 = Bind Event to Cosmetic Item Changed(self:Sparks Cosmetic Component Object Reference=@n3, Delegate=Create Event()) {is=title K2Node_AddDelegate}
   ReloadData(execute=[@n5.then, @n4.Is Not Valid])
   Event Construct()
  ]}
 GetChoice {nodes [
   n1 = GetChoice!() -> (
    then
    Item Subtype:ESparksAccountItemSubtype Enum=Aura
    Instrument type:ESparksInstrumentType Enum=Guitar
    Variant Channel:Gameplay Tag Structure
   ) {is=title K2Node_FunctionEntry}
   n2 = Get CosmeticComponent() -> (CosmeticComponent:Sparks Cosmetic Component Object Reference)
   > Return Node!(
    execute
    Choice:String="To String (Boolean)"(InBool=GetChoice!(
      self:Sparks Cosmetic Component Object Reference=@n2
      Subtype:ESparksAccountItemSubtype Enum=@n1
      InstrumentType:ESparksInstrumentType Enum=@n1.Instrument type
      VariantChannel:Gameplay Tag Structure=@n1.Variant Channel
     ) -> (OutChannel:Mcp Variant Channel Info Structure, ReturnValue:bool) {is=title K2Node_CallFunction}.ReturnValue)
   )
  ]}
 OnCosmeticChoiceChanged {nodes [
   n1 = OnCosmeticChoiceChanged!() -> (then, Subtype:ESparksAccountItemSubtype Enum=Aura, InstrumentType:ESparksInstrumentType Enum=Guitar) {is=title K2Node_FunctionEntry}
   > n2 = Switch on ESparksAccountItemSubtype!(execute, Selection:ESparksAccountItemSubtype Enum=@n1) -> (Aura:Exec, Guitar:Exec, Bass:Exec, Keyboard:Exec, Microphone:Exec, Drums:Exec, SpotlightAnim:Exec, Song:Exec) {is=title K2Node_SwitchEnum}
   > RefreshChoices(self:WBP Sparks Dbg Cosmetic Item Object Reference=Get AuraWidget() -> (AuraWidget:WBP Sparks Dbg Cosmetic Item Object Reference))
   OnChoiceChanged!(
    execute=@n2.SpotlightAnim
    self:WBP Sparks Dbg Cosmetic Group Object Reference=Get SpotlightWidget() -> (SpotlightWidget:WBP Sparks Dbg Cosmetic Group Object Reference)
    Instrument:ESparksInstrumentType Enum=@n1.InstrumentType
   ) -> (then)
   OnChoiceChanged(
    execute=[@n2.Bass, @n2.Drums, @n2.Guitar, @n2.Keyboard, @n2.Microphone]
    self=Get InstrumentWidget() -> (InstrumentWidget:WBP Sparks Dbg Cosmetic Group Object Reference)
    Instrument=@n1.InstrumentType
   )
  ]}
 OnCosmeticItemChanged {nodes [
   n1 = OnCosmeticItemChanged!() -> (then, ItemDef:Sparks Account Item Definition Object Reference) {is=title K2Node_FunctionEntry}
   > n2 = Switch on ESparksAccountItemSubtype(Selection=Get SubType!(self:Sparks Account Item Definition Object Reference=@n1) -> (SubType:ESparksAccountItemSubtype Enum=Aura))
   > SetItemInfo!(execute, self:WBP Sparks Dbg Cosmetic Item Object Reference=Get AuraWidget(), Item:Sparks Account Item Definition Object Reference=@n1) -> (then)
   SetItemInfo!(
    execute=@n2.SpotlightAnim
    self:WBP Sparks Dbg Cosmetic Group Object Reference=Get SpotlightWidget()
    NewItem:Sparks Account Item Definition Object Reference=@n1
   ) -> (then)
   SetItemInfo(execute=[@n2.Bass, @n2.Drums, @n2.Guitar, @n2.Keyboard, @n2.Microphone], self=Get InstrumentWidget(), NewItem=@n1)
  ]}
 ReloadData {nodes [
   n1 = Create Widget!(
    execute:Exec
    Class:User Widget Class Reference=&/WBP_Sparks_Dbg_Cosmetic_Group.WBP_Sparks_Dbg_Cosmetic_Group_C
    OwningPlayer:Player Controller Object Reference
   ) -> (then:Exec, ReturnValue:WBP Sparks Dbg Cosmetic Group Object Reference)
   n2 = Get vb_Items()
   > AddChild(self=@n2, Content=@n1)
   n3 = Self-Reference()
   > Set ParentPicker!(execute, ParentPicker:WBP Sparks Dbg Cosmetic Picker Object Reference=@n3, self:WBP Sparks Dbg Cosmetic Group Object Reference=@n1) -> (then, Output_Get:WBP Sparks Dbg Cosmetic Picker Object Reference)
   > n4 = Set SpotlightWidget(SpotlightWidget:WBP Sparks Dbg Cosmetic Group Object Reference=@n1) -> (Output_Get:WBP Sparks Dbg Cosmetic Group Object Reference)
   > SetGroupSubtype!(execute, self:WBP Sparks Dbg Cosmetic Group Object Reference=@n4, ItemType:ESparksAccountItemSubtype Enum=SpotlightAnim) -> (then)
   ReloadData() {is=title K2Node_FunctionEntry}
   > ClearChildren(self=Get vb_Items())
   > n5 = Sequence!(execute) -> (then_0, then_1, then_2:Exec)
   > n6 = Create Widget(Class=&/WBP_Sparks_Dbg_CosmeticItem.WBP_Sparks_Dbg_CosmeticItem_C) -> (ReturnValue:WBP Sparks Dbg Cosmetic Item Object Reference)
   > AddChild(self=@n2, Content=@n6)
   > Set ParentPicker(ParentPicker=@n3, self:WBP Sparks Dbg Cosmetic Item Object Reference=@n6)
   > n7 = Set AuraWidget(AuraWidget:WBP Sparks Dbg Cosmetic Item Object Reference=@n6) -> (Output_Get:WBP Sparks Dbg Cosmetic Item Object Reference)
   > SetTypes!(
    execute
    self:WBP Sparks Dbg Cosmetic Item Object Reference=@n7
    Subtype:ESparksAccountItemSubtype Enum=Aura
    Instrument:ESparksInstrumentType Enum=Guitar
   ) -> (then)
   n8 = Create Widget(execute=@n5.then_1, Class=&WBP_Sparks_Dbg_Cosmetic_Group_C) -> (ReturnValue:WBP Sparks Dbg Cosmetic Group Object Reference)
   > AddChild(self=@n2, Content=@n8)
   > Set ParentPicker(ParentPicker=@n3, self:WBP Sparks Dbg Cosmetic Group Object Reference=@n8)
   > n9 = Set InstrumentWidget(InstrumentWidget:WBP Sparks Dbg Cosmetic Group Object Reference=@n8) -> (Output_Get:WBP Sparks Dbg Cosmetic Group Object Reference)
   > SetGroupSubtype(self=@n9, ItemType=Guitar)
  ]}
]
variables [
 AuraWidget {type=WBP_Sparks_Dbg_CosmeticItem_C, default=null}
 CosmeticComponent {type=SparksCosmeticComponent, default=null}
 InstrumentWidget {type=WBP_Sparks_Dbg_Cosmetic_Group_C, default=null}
 SpotlightWidget {type=WBP_Sparks_Dbg_Cosmetic_Group_C, default=null}
]
widgets {
 info {
  inheritedWidgetCount=0
  parentClass=/Script/SparksCosmeticsUI.SparksCosmeticsDebugPickerWidget
  rootWidgetClass=&CanvasPanel
  widgetCount=12
 }
 widgets [
  CanvasPanel_138 {widgetClassPath=&CanvasPanel, widget=CanvasPanel_138}
  CanvasPanel_139 {
   widgetClassPath=&CanvasPanel
   parent=CanvasPanel_138
   widget=CanvasPanel_139
   slot {
    class=CanvasPanelSlot
    properties {bAutoSize=true, layoutData {alignment=<0 0>, anchors {maximum=<1 1>, minimum=<0 0>}, offsets=<ltrb 1281.9219 32 40 993.0651>}}
   }
  }
  VerticalBox_0 {
   widgetClassPath=&VerticalBox
   parent=CanvasPanel_139
   widget=VerticalBox_0
   slot {
    class=CanvasPanelSlot
    properties {bAutoSize=true, layoutData {alignment=<0 0>, anchors {maximum=<1 0>, minimum=<0 0>}, offsets=<ltrb 0 0 0 250>}}
   }
  }
  HorizontalBox_0 {widgetClassPath=&HorizontalBox, parent=VerticalBox_0, widget=HorizontalBox_0, slot {class=VerticalBoxSlot}}
  btn_Reload {
   widgetClassPath=&IconTextButton
   bIsVariable=true
   parent=HorizontalBox_0
   widget=btn_Reload
   properties {
    buttonText=Reload
    holdData=&CommonUI_HoldData_Base_C
    onButtonBaseClicked=(&self:WidgetTree.btn_Reload.HandleOnButtonClicked)
    onButtonBaseDoubleClicked=(&self:WidgetTree.btn_Reload.HandleOnButtonDoubleClicked)
    onButtonBaseHovered=(&self:WidgetTree.btn_Reload.HandleOnButtonHovered)
    onButtonBaseUnhovered=(&self:WidgetTree.btn_Reload.HandleOnButtonUnhovered)
    onSelectedChangedBase=(&self:WidgetTree.btn_Reload.HandleOnSelectedChanged)
   }
   slot {class=HorizontalBoxSlot, properties {padding=<ltrb 3 3 3 3>}}
  }
  btn_Save {
   widgetClassPath=&IconTextButton
   bIsVariable=true
   parent=HorizontalBox_0
   widget=btn_Save
   properties {
    buttonText=Save
    holdData=&CommonUI_HoldData_Base_C
    onButtonBaseClicked=(&self:WidgetTree.btn_Save.HandleOnButtonClicked)
    onButtonBaseDoubleClicked=(&self:WidgetTree.btn_Save.HandleOnButtonDoubleClicked)
    onButtonBaseHovered=(&self:WidgetTree.btn_Save.HandleOnButtonHovered)
    onButtonBaseUnhovered=(&self:WidgetTree.btn_Save.HandleOnButtonUnhovered)
    onSelectedChangedBase=(&self:WidgetTree.btn_Save.HandleOnSelectedChanged)
   }
   slot {class=HorizontalBoxSlot, properties {padding=<ltrb 3 3 3 3>}}
  }
  btn_Defaults {
   widgetClassPath=&IconTextButton
   bIsVariable=true
   parent=HorizontalBox_0
   widget=btn_Defaults
   properties {
    buttonText=Defaults
    holdData=&CommonUI_HoldData_Base_C
    onButtonBaseClicked=(&self:WidgetTree.btn_Defaults.HandleOnButtonClicked)
    onButtonBaseDoubleClicked=(&self:WidgetTree.btn_Defaults.HandleOnButtonDoubleClicked)
    onButtonBaseHovered=(&self:WidgetTree.btn_Defaults.HandleOnButtonHovered)
    onButtonBaseUnhovered=(&self:WidgetTree.btn_Defaults.HandleOnButtonUnhovered)
    onSelectedChangedBase=(&self:WidgetTree.btn_Defaults.HandleOnSelectedChanged)
   }
   slot {class=HorizontalBoxSlot, properties {padding=<ltrb 3 3 3 3>}}
  }
  chk_Collapse {
   widgetClassPath=/Script/UMG.CheckBox
   bIsVariable=true
   parent=HorizontalBox_0
   widget=chk_Collapse
   properties {checkedState=Checked}
   slot {class=HorizontalBoxSlot, properties {horizontalAlignment=HAlign_Right, padding=<ltrb 15 15 15 15>, size {sizeRule=Fill, value=1}}}
  }
  CanvasPanel_0 {widgetClassPath=&CanvasPanel, parent=VerticalBox_0, widget=CanvasPanel_0, slot {class=VerticalBoxSlot}}
  vb_Items {
   widgetClassPath=&VerticalBox
   bIsVariable=true
   parent=CanvasPanel_0
   widget=vb_Items
   slot {
    class=CanvasPanelSlot
    properties {bAutoSize=true, layoutData {alignment=<0 0>, anchors {maximum=<1 1>, minimum=<0 0>}, offsets=<ltrb 0 0 0 -62.0621>}}
   }
  }
  Image {
   widgetClassPath=&Image
   bIsVariable=true
   parent=CanvasPanel_0
   widget=Image
   properties {brush {
     bIsDynamicallyLoaded=false
     drawAs=Image
     imageType=NoImage
     mirroring=NoMirror
     resourceName=None
     resourceObject=null
     tiling=NoTile
     imageSize=<32 32>
     margin=<ltrb 0 0 0 0>
     outlineSettings {
      bUseBrushTransparency=false
      roundingType=HalfHeightRadius
      width=0
      color {colorUseRule=UseColor_Specified, specifiedColor=<rgba 0 0 0 0>}
      cornerRadii=<0 0 0 0>
     }
     tintColor {colorUseRule=UseColor_Specified, specifiedColor=<rgba 0 0 0 0.2>}
     uVRegion {bIsValid=false, max=<0 0>, min=<0 0>}
    }}
   slot {
    class=CanvasPanelSlot
    properties {bAutoSize=true, layoutData {alignment=<0 0>, anchors {maximum=<1 1>, minimum=<0 0>}, offsets=<ltrb 0 0 0 -62.0621>}}
   }
  }
  Image_150 {
   widgetClassPath=&Image
   bIsVariable=true
   parent=CanvasPanel_139
   widget=Image_150
   properties {brush {
     bIsDynamicallyLoaded=false
     drawAs=Image
     imageType=NoImage
     mirroring=NoMirror
     resourceName=None
     resourceObject=null
     tiling=NoTile
     imageSize=<32 32>
     margin=<ltrb 0 0 0 0>
     outlineSettings {
      bUseBrushTransparency=false
      roundingType=HalfHeightRadius
      width=0
      color {colorUseRule=UseColor_Specified, specifiedColor=<rgba 0 0 0 0>}
      cornerRadii=<0 0 0 0>
     }
     tintColor {colorUseRule=UseColor_Specified, specifiedColor=<rgba 0 0 0 0.2>}
     uVRegion {bIsValid=false, max=<0 0>, min=<0 0>}
    }}
   slot {class=CanvasPanelSlot, properties {layoutData {alignment=<0 0>, anchors {maximum=<1 1>, minimum=<0 0>}, offsets=<ltrb 2 2 2 2>}}}
  }
 ]
}
